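; Generic Fortnite Athena Blueprint for dual-wielded ranged weapons, handling cosmetic fire, reload and scope behavior.
; Construction attaches Cascade or Niagara muzzle effects to both hand meshes, plus an optional wind particle.
; Hand-fired events drive per-hand muzzle, shell and rumble feedback, with timers enforcing minimum looping-FX time and force-stopping stuck loops.
; It also handles reload smoke, scope post-process with mesh swapping and enemy custom depth, destruction effects and pierce-through toggling.
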(event EventBeginPlay
  (|CallParentFunction)
  (Utilities|FlowControl|Sequence
    (:then_0
      (CallFunction|BindonHandFired)
      ;[0 0 0] Send wind particle reference to the wind manager
      (if (Variables|Default|GetIs%20Wind%20Enabled)
        (Wind|AddWindParticleSystemComponent (Fortnite|GetWindManager) (Variables|Effects|GetMuzzleWindParticleSpawned))))
    (:then_1
      (if (Variables|Default|GetDisableAnimation)
        (FortWeapon|SetShouldTickWeaponMesh false)
        (Components|SkinnedMesh|SetRenderStatic (Visual|GetWeaponMesh) true)))))

(event Custom|BindOnHandFired
  ;[1 1 1] On hand fired (left - right)
  (RangedWeaponDual|BindEventtoOnHandFired (delegate Custom|HandFired)))

(event Custom|ShellsON_(onPump)
  ;[0.3841 0.6674 1] Shells VFX (on Pump)
  (CallFunction|ActivateShellsFX (and (Variables|Effects%28Shells%29|GetUseShellsOnPump?) (> (Rendering|GetRenderingDetailMode) 0))))

(event Custom|HandFired
  (params (Hand :type EDualWeaponHand%20Enum :default "LEFT") (bPersistantFire :type Boolean))
  ;[1 1 1] On hand fired (left - right)
  (bind _output_get (Variables|Default|SetLastHandFired Hand))
  (CallFunction|MuzzleFlashFX bPersistantFire (Utilities|Enum|Equal(Enum) _output_get "LEFT"))
  (if (and (not (Weapon|IsUsingScopeorFirstPersonCamera)) (> (Rendering|GetRenderingDetailMode) 0))
    (CallFunction|ActivateShellsFX (Variables|Effects%28Shells%29|GetUseShellsOnFire?) (Utilities|Enum|Equal(Enum) _output_get "LEFT")))
  ;[1 1 1] Comment
  (FortWeapon|PlayForceFeedback false (Variables|Default|GetbForceFeedbackShouldPersist))
  ;[1 0.09001 0.04025] Failsafe hack to stop looping FX if we exceed the RPC limit and don't properly remove the gameplay cue
  (if (Effects|ShouldPlayPersistentFireFX)
    (Variables|Default|SetbStoppedWeaponFireFX false)
    (Utilities|Time|ClearandInvalidateTimerbyHandle (Variables|Default|GetLoopingFireFXFailsafeTimerHandle))
    (bind _returnvalue (Utilities|Time|SetTimerbyFunctionName 0 "ForceStopWeaponFireFX" (Math|Float|SelectFloat (Variables|Effects|GetOverrideFailsafeStopLoopingFXTimer) (+ (* (/ (Math|Conversions|ToFloat(Integer) (FortWeapon|GetLocalMagazineAmmoCount)) (Math|Float|Max(Float) (FortWeapon|GetFiringRate) 0.1)) 1.2) 0.0001) (> (Variables|Effects|GetOverrideFailsafeStopLoopingFXTimer) 0))))
    (Variables|Default|SetLoopingFireFXFailsafeTimerHandle _returnvalue)))

(event EventOnPlayWeaponFireFX
  (params (bPersistentFire :type Boolean) (bSecondaryFire :type Boolean))
  ;[0.4433 0 1] Muzzle Flash VFX
  (Utilities|FlowControl|Sequence
    (:then_0
      (CallFunction|ActivateorDeactivateWindParticle true))
    (:then_1
      (FortWeapon|PlayForceFeedback false bPersistentFire))))

(event Custom|ShellEjectionFixOn
  ;[1 1 1] Fix for Shell Ejection Issue on Quality Change
  (EventDispatchers|BindEventtoOnEffectsQualityChanged (Settings|GetFortGameUserSettings) (delegate Custom|Bind%20on%20Effects%20Quality)))

(event Custom|ShellEjectionOff
  ;[1 1 1] Fix for Shell Ejection Issue on Quality Change
  (EventDispatchers|UnbindEventfromOnEffectsQualityChanged (Settings|GetFortGameUserSettings) (delegate Custom|Bind%20on%20Effects%20Quality)))

(event Custom|Bind%20on%20Effects%20Quality
  ;[1 1 1] Fix for Shell Ejection Issue on Quality Change
  (Utilities|FlowControl|Delay 0.1) ; REQUIRED to make work!
  (for _array_element (Utilities|Array|MakeArray (Variables|Effects%28Shells%29|GetSpawnedShellsPS_R) (Variables|Effects%28Shells%29|GetSpawnedShellsPS_L) (Variables|Default|GetSpawned%20Shells%20Niagara%20L) (Variables|Default|GetSpawned%20Shells%20Niagara%20R))
    (CallFunction|UpdateShellEmittersFX _array_element)))

(event EventOnStopWeaponFireFX
  (bind _returnvalue (Utilities|Time|GetGameTimeinSeconds))
  (Default|CallOnPersistantFireStopped)
  ;[0.4433 0 1] Muzzle Flash VFX
  (Variables|Default|SetbStoppedWeaponFireFX true)
  (Utilities|Time|ClearTimerbyFunctionName 0 "StopWeaponFireFX")
  ;;[1 1 1] Enforce a short minimum time to play looping FX
  (Utilities|Time|ClearandInvalidateTimerbyHandle (Variables|Default|GetDelayedLoopingFXEndHandle))
  (if (>= (- _returnvalue (Variables|Effects|GetLastPlayFXTime)) (Variables|Effects|GetMinPlayFXTime))
    (CallFunction|ActivateorDeactivateWindParticle false)
    (CallFunction|DeactivateMuzzleFX)
    (Effects|DeactivateBeamFX)
    (CallFunction|DeactivateShellsFX)
    (FortWeapon|StopForceFeedback)
    (else
      (bind _returnvalue_2 (Utilities|Time|SetTimerbyEvent (delegate Custom|Delayed%20Looping%20FX%20End) (- (Variables|Effects|GetMinPlayFXTime) (- _returnvalue (Variables|Effects|GetLastPlayFXTime)))))
      (Variables|Default|SetDelayedLoopingFXEndHandle _returnvalue_2))))

(event Custom|ForceStopWeaponFireFX
  (bind _returnvalue (Utilities|Time|GetGameTimeinSeconds))
  (Networking|CanExecuteCosmeticEvents
    (:True
      (FortWeapon|StopWeaponFireFX)
      (if (not (Variables|Default|GetbStoppedWeaponFireFX))
        (Default|CallOnPersistantFireStopped)
        ;[0.4433 0 1] Muzzle Flash VFX
        (Variables|Default|SetbStoppedWeaponFireFX true)
        (Utilities|Time|ClearTimerbyFunctionName 0 "StopWeaponFireFX")
        ;;[1 1 1] Enforce a short minimum time to play looping FX
        (Utilities|Time|ClearandInvalidateTimerbyHandle (Variables|Default|GetDelayedLoopingFXEndHandle))
        (if (>= (- _returnvalue (Variables|Effects|GetLastPlayFXTime)) (Variables|Effects|GetMinPlayFXTime))
          (CallFunction|ActivateorDeactivateWindParticle false)
          (CallFunction|DeactivateMuzzleFX)
          (Effects|DeactivateBeamFX)
          (CallFunction|DeactivateShellsFX)
          (FortWeapon|StopForceFeedback)
          (else
            (bind _returnvalue_2 (Utilities|Time|SetTimerbyEvent (delegate Custom|Delayed%20Looping%20FX%20End) (- (Variables|Effects|GetMinPlayFXTime) (- _returnvalue (Variables|Effects|GetLastPlayFXTime)))))
            (Variables|Default|SetDelayedLoopingFXEndHandle _returnvalue_2)))))))

(event Custom|Delayed%20Looping%20FX%20End
  ;[0.4433 0 1] Muzzle Flash VFX
  (CallFunction|ActivateorDeactivateWindParticle false)
  (CallFunction|DeactivateMuzzleFX)
  (Effects|DeactivateBeamFX)
  (CallFunction|DeactivateShellsFX)
  (FortWeapon|StopForceFeedback))

(event EventOnPlayReloadFX
  (params (ReloadStage :type EFortReloadFXState%20Enum :default "ReloadStart"))
  ;[0.3518 0.0375 0.63] Use Reload Smoke Trails
  (CallFunction|MuzzlePlayReloadFX ReloadStage)
  (FortWeapon|PlayForceFeedback true))

(event FortWeapon|EventOnEquippedAndAttached
  (bind _scopemesh1p (Variables|Components|GetScopeMesh1P))
  ;[1 1 1] OnEquippedAndAttached()
  (|CallParentFunction)
  (CallFunction|ShowReticle)
  (Utilities|FlowControl|Sequence
    (:then_0
      ;[1 1 1] Init Scope Variables
      (bind _asfort_pawn (Utilities|Casting|CastToFortPawn (Game|GetInstigator)))
      (Utilities|IsValid (Pawn|GetLocalViewingPlayerController _asfort_pawn)
        (:Is%20Valid
          (Networking|CanExecuteCosmeticEvents
            (:True
              (if (Scope|UseScopeTargeting)
                (Utilities|IsValid _scopemesh1p
                  (:Is%20Valid
                    (Rendering|Material|SetVectorParameterValue "/Game/Packages/Fortress_SharedMaterials/GlobalMaterialParameters/FortniteMaterialParameters.FortniteMaterialParameters" "TopLeftScope" (Math|Conversions|ToLinearColor(Vector) (Transformation|GetSocketLocation _scopemesh1p "TopLeftScope")) :!class /Script/Engine.KismetMaterialLibrary)
                    (Rendering|Material|SetVectorParameterValue "/Game/Packages/Fortress_SharedMaterials/GlobalMaterialParameters/FortniteMaterialParameters.FortniteMaterialParameters" "BottomRightScope" (Math|Conversions|ToLinearColor(Vector) (Transformation|GetSocketLocation _scopemesh1p "BottomRightScope")) :!class /Script/Engine.KismetMaterialLibrary)))))))))
    (:then_1
      (Rendering|SetReceivesDecals (CallFunction|GetLeftHandWeaponMesh) false) ; So Sprays don't show up on weapon
    )))

(event FortWeapon|EventOnInstigatorSet
  (bind _scopemesh1p (Variables|Components|GetScopeMesh1P))
  ;[1 1 1] OnEquippedAndAttached()
  (|CallParentFunction)
  (CallFunction|ShowReticle)
  (Utilities|FlowControl|Sequence
    (:then_0
      ;[1 1 1] Init Scope Variables
      (bind _asfort_pawn (Utilities|Casting|CastToFortPawn (Game|GetInstigator)))
      (Utilities|IsValid (Pawn|GetLocalViewingPlayerController _asfort_pawn)
        (:Is%20Valid
          (Networking|CanExecuteCosmeticEvents
            (:True
              (if (Scope|UseScopeTargeting)
                (Utilities|IsValid _scopemesh1p
                  (:Is%20Valid
                    (Rendering|Material|SetVectorParameterValue "/Game/Packages/Fortress_SharedMaterials/GlobalMaterialParameters/FortniteMaterialParameters.FortniteMaterialParameters" "TopLeftScope" (Math|Conversions|ToLinearColor(Vector) (Transformation|GetSocketLocation _scopemesh1p "TopLeftScope")) :!class /Script/Engine.KismetMaterialLibrary)
                    (Rendering|Material|SetVectorParameterValue "/Game/Packages/Fortress_SharedMaterials/GlobalMaterialParameters/FortniteMaterialParameters.FortniteMaterialParameters" "BottomRightScope" (Math|Conversions|ToLinearColor(Vector) (Transformation|GetSocketLocation _scopemesh1p "BottomRightScope")) :!class /Script/Engine.KismetMaterialLibrary)))))))))
    (:then_1
      (Rendering|SetReceivesDecals (CallFunction|GetLeftHandWeaponMesh) false) ; So Sprays don't show up on weapon
    )))

(event Custom|InitializeScopeVariables
  (bind _scopemesh1p (Variables|Components|GetScopeMesh1P))
  ;[1 1 1] Init Scope Variables
  (bind _asfort_pawn (Utilities|Casting|CastToFortPawn (Game|GetInstigator)))
  (Utilities|IsValid (Pawn|GetLocalViewingPlayerController _asfort_pawn)
    (:Is%20Valid
      (Networking|CanExecuteCosmeticEvents
        (:True
          (if (Scope|UseScopeTargeting)
            (Utilities|IsValid _scopemesh1p
              (:Is%20Valid
                (Rendering|Material|SetVectorParameterValue "/Game/Packages/Fortress_SharedMaterials/GlobalMaterialParameters/FortniteMaterialParameters.FortniteMaterialParameters" "TopLeftScope" (Math|Conversions|ToLinearColor(Vector) (Transformation|GetSocketLocation _scopemesh1p "TopLeftScope")) :!class /Script/Engine.KismetMaterialLibrary)
                (Rendering|Material|SetVectorParameterValue "/Game/Packages/Fortress_SharedMaterials/GlobalMaterialParameters/FortniteMaterialParameters.FortniteMaterialParameters" "BottomRightScope" (Math|Conversions|ToLinearColor(Vector) (Transformation|GetSocketLocation _scopemesh1p "BottomRightScope")) :!class /Script/Engine.KismetMaterialLibrary)))))))))

(event Custom|UnhideThirdPersonStuff
  ;[1 1 1] Scoped weapon targeting downsights
  ;;[1 1 1] Hiding/showing 3p meshes, etc
  (if (not (Weapon|IsUsingFirstPersonCamera))
    (bind _asfort_pawn (Utilities|Casting|CastToFortPawn (Game|GetInstigator)))
    (FortPawn|SetPawnVisibilityLocalControllerOnly _asfort_pawn false false)
    (FortWeapon|SetHiddeninGameControllerOnly (Variables|Visual|GetWeaponMesh)))
  (CallFunction|ShowReticle)
  (bind _asfort_player_controller_zone (Utilities|Casting|CastToFortPlayerControllerZone (Game|GetInstigatorController)))
  (Class|FortPlayerController|FixupInteractionWidgetsonUnzoom _asfort_player_controller_zone))

(event Custom|HideFirstPersonStuff
  ;[1 1 1] Scoped weapon targeting downsights
  ;;[1 1 1] Hiding/showing 1p meshes, etc
  (Development|SetHiddeninGame (Variables|Components|GetScopeMesh1P) true false)
  (Variables|Scope|SetScopePostProcessEnabled false)
  ;;;[1 1 1] Custom Depth
  (if (Variables|Scope|GetScope%20-%20Render%20Enemies%20To%20Custom%20Depth%20Buffer)
    (CallFunction|UpdateEnemyCustomDepths false 0)))

(event EventOnSetTargeting
  (params (bNewIsTargeting :type Boolean))
  ;[1 1 1] Scoped weapon targeting downsights
  (Networking|CanExecuteCosmeticEvents
    (:True
      (bind _asfort_pawn (Utilities|Casting|CastToFortPawn (Game|GetInstigator)))
      (Utilities|IsValid (Pawn|GetLocalViewingPlayerController _asfort_pawn)
        (:Is%20Valid
          (if (Scope|UseScopeTargeting)
            (if bNewIsTargeting
              (CallFunction|InitializeScopeVariables)
              (Utilities|FlowControl|Sequence
                (:then_0
                  (Utilities|FlowControl|Delay 0.1)
                  (if (Camera|GetIsTargeting)
                    (Utilities|FlowControl|Sequence
                      (:then_0
                        (CallFunction|UnhideFirstPersonStuffPart1)
                        (CallFunction|PlayScopePP))
                      (:then_1
                        (Utilities|FlowControl|DoOnce false
                          (:Completed
                            (Audio|SpawnSoundAttached (Variables|Sounds|GetSound_ScopeZoomIn) (Variables|Visual|GetWeaponMesh))))))))
                (:then_1
                  (Utilities|FlowControl|Delay 0.2)
                  (if (Camera|GetIsTargeting)
                    (CallFunction|HideThirdPersonStuff)
                    (CallFunction|UnhideFirstPersonStuffPart2))))
              (else
                (Utilities|FlowControl|Sequence
                  (:then_0
                    (CallFunction|UnhideThirdPersonStuff))
                  (:then_1
                    (Audio|SpawnSoundAttached (Variables|Sounds|GetSound_ScopeZoomOut) (Variables|Visual|GetWeaponMesh))
                    (CallFunction|ReverseScopePP))
                  (:then_2
                    (CallFunction|ResetDoonceScopeSound)))))
            (else
              (CallFunction|ShowReticle))))))))

(event Custom|ResetDoonceScopeSound
  ;[1 1 1] Scoped weapon targeting downsights
  (Utilities|FlowControl|DoOnce false
    (:Completed
      (Audio|SpawnSoundAttached (Variables|Sounds|GetSound_ScopeZoomIn) (Variables|Visual|GetWeaponMesh)))))

(event Custom|UnhideFirstPersonStuffPart2
  (bind _scopemesh1p (Variables|Components|GetScopeMesh1P))
  ;[1 1 1] Scoped weapon targeting downsights
  ;;[1 1 1] Hiding/showing 1p meshes, etc
  (Development|SetHiddeninGame _scopemesh1p false false)
  (FortWeapon|SetHiddeninGameOtherLocalControllersOnly _scopemesh1p true)
  ;;;[1 1 1] Custom Depth
  (if (Variables|Scope|GetScope%20-%20Render%20Enemies%20To%20Custom%20Depth%20Buffer)
    (CallFunction|UpdateEnemyCustomDepths true (Variables|Default|GetStencilBufferValue))))

(event Custom|HideThirdPersonStuff
  ;[1 1 1] Scoped weapon targeting downsights
  ;;[1 1 1] Hiding/showing 3p meshes, etc
  (bind _asfort_pawn (Utilities|Casting|CastToFortPawn (Game|GetInstigator)))
  (FortPawn|SetPawnVisibilityLocalControllerOnly _asfort_pawn true false)
  (FortWeapon|SetWeaponVisibilityControllerOnly true true)
  (FortWeapon|SetHiddeninGameControllerOnly (Variables|Visual|GetWeaponMesh) true))

(event FortWeapon|EventOnLocalUnEquip
  ;[1 1 1] Scoped weapon targeting downsights
  (bind _asfort_pawn (Utilities|Casting|CastToFortPawn (Game|GetInstigator)))
  (Utilities|IsValid (Pawn|GetLocalViewingPlayerController _asfort_pawn)
    (:Is%20Valid
      (if (Scope|UseScopeTargeting)
        (Utilities|FlowControl|Sequence
          (:then_0
            (CallFunction|UnhideThirdPersonStuff))
          (:then_1
            (Audio|SpawnSoundAttached (Variables|Sounds|GetSound_ScopeZoomOut) (Variables|Visual|GetWeaponMesh))
            (CallFunction|ReverseScopePP))
          (:then_2
            (CallFunction|ResetDoonceScopeSound)))
        (else
          (CallFunction|ShowReticle))))))

(event Custom|PlayScopePP
  ;[1 1 1] Scoped weapon targeting downsights
  ;;[1 1 1] Scope Post Process
  (bind (_direction _downsightpostprocessamount) (Timeline|AnimateScopePostProcess Play)
    (:Update
      (CallFunction|SetPostProcessParams _downsightpostprocessamount))
    (:Finished
      (if (Utilities|Enum|Equal(Enum) _direction "Backward")
        (if (not (Camera|GetIsTargeting))
          (CallFunction|HideFirstPersonStuff))))))

(event Custom|UnhideFirstPersonStuffPart1
  ;[1 1 1] Scoped weapon targeting downsights
  ;;[1 1 1] Hiding/showing 1p meshes, etc
  (Variables|Scope|SetScopePostProcessEnabled true))

(event Custom|ReverseScopePP
  ;[1 1 1] Scoped weapon targeting downsights
  ;;[1 1 1] Scope Post Process
  (Timeline|AnimateScopePostProcess Reverse))

(event Targeting|EventAbortScopeFX
  ;[1 1 1] Direct control of scoping, useful when switching between spectated characters
  (Utilities|FlowControl|Sequence
    (:then_0
      (CallFunction|SetPostProcessParams 0.0))
    (:then_1
      (CallFunction|HideFirstPersonStuff))
    (:then_2
      (CallFunction|UnhideThirdPersonStuff))))

(event Targeting|EventForceScopeFX
  ;[1 1 1] Direct control of scoping, useful when switching between spectated characters
  (Utilities|FlowControl|Sequence
    (:then_0
      (CallFunction|SetPostProcessParams 1.0))
    (:then_1
      (CallFunction|UnhideFirstPersonStuffPart1)
      (CallFunction|UnhideFirstPersonStuffPart2))
    (:then_2
      (CallFunction|HideThirdPersonStuff))))

(event Custom|Update%20Enemy%20Custom%20Depths
  (params (Enable%20Or%20Disable :type Boolean :default true) (StencilBufferValue :type Integer :default 3))
  ;[1 1 1] Custom Depth
  (bind _outactors (Actor|GetAllActorsOfClass "/Script/FortniteGame.FortAIPawn"))
  (bind _output_get (Variables|Scope|SetArray%20Of%20Active%20Enemy%20AI _outactors))
  (for _array_element _output_get
    (Utilities|IsValid (Variables|/Script/Engine.Character|GetMesh _array_element)
      (:Is%20Valid
        (Rendering|SetRenderCustomDepth (Variables|/Script/Engine.Character|GetMesh _array_element) Enable%20Or%20Disable)
        (Rendering|SetCustomDepthStencilValue (Variables|/Script/Engine.Character|GetMesh _array_element) StencilBufferValue)))))

(event EventOnEquippedWeaponDestory
  ;[0.725 0.6854 0.4277] Show fx for weapon destruction on depleted durability
  (if (Variables|Effects|GetUseDestroyEffect)
    (Effects|SpawnEmitteratLocation (Variables|Effects|GetWeaponDurabilityDestroyEffect) (Transformation|GetSocketLocation (Variables|Visual|GetWeaponMesh) "Handle"))
    (bind _asfort_pawn (Utilities|Casting|CastToFortPawn (Game|GetInstigator)))
    (bind _returnvalue (Effects|SpawnEmitterAttached :EmitterTemplate (Variables|Effects|GetWeaponDurabilityDestroyEffectIcon) :AttachToComponent (Variables|/Script/Engine.Character|GetMesh _asfort_pawn) :AttachPointName "Root_Effect" :Location "0, 0,130.000000" :Scale "1.000000,1.000000,1.000000" :LocationType "SnapToTarget" :bAutoDestroy true :PoolingMethod "None" :bAutoActivate true))
    (bind _returnvalue_2 (Rendering|Material|CreateDynamicMaterialInstance _returnvalue 0 "/Game/Effects/WeaponDestroy/M_WeaponDestroyExclamation.M_WeaponDestroyExclamation" "None" :!class /Script/Engine.PrimitiveComponent))
    (Rendering|Material|SetMaterial _returnvalue 0 _returnvalue_2)
    (Rendering|Material|SetTextureParameterValue _returnvalue_2 "IconTex" (ItemPreviewImage|GetorLoadSynchronously (Item|GetWeaponDataSmallPreviewImage)))))

(event Custom|SetWeaponPierceThrough
  (params (Enable :type Boolean) (TargetLimit :type Integer))
  (bind _returnvalue (Math|Integer|SelectInt TargetLimit 0 Enable))
  ;[1 1 1] Event to dynamically enable/disable trace through pawns
  (FortWeapon|SetTraceThroughPawnsLimit _returnvalue)
  (FortWeapon|SetTraceThroughPawns Enable)
  (CallFunction|SetWeaponPierceThroughClientRep Enable _returnvalue))

(event Custom|SetWeaponPierceThrough_ClientRep
  (params (Enable :type Boolean) (TargetLimit :type Integer))
  ;[1 1 1] Event to dynamically enable/disable trace through pawns
  (FortWeapon|SetTraceThroughPawnsLimit (Math|Integer|SelectInt TargetLimit 0 Enable))
  (FortWeapon|SetTraceThroughPawns Enable))

(event Custom|HideWeaponMesh
  ;[1 1 1] DEBUG EVENTS -> ke * <event name>
  (Development|SetHiddeninGame (Variables|Visual|GetWeaponMesh) true false)
  (Development|PrintString "HideWeaponMesh"))

(event Custom|ShowWeaponMesh
  ;[1 1 1] DEBUG EVENTS -> ke * <event name>
  (Development|SetHiddeninGame (Variables|Visual|GetWeaponMesh) false false)
  (Development|PrintString "ShowWeaponMesh"))

(event Custom|HideWeapon
  ;[1 1 1] DEBUG EVENTS -> ke * <event name>
  (FortWeapon|SetWeaponVisibility false true)
  (Development|PrintString "HideWeapon"))

(event Custom|ShowWeapon
  ;[1 1 1] DEBUG EVENTS -> ke * <event name>
  (FortWeapon|SetWeaponVisibility true true)
  (Development|PrintString "ShowWeapon"))

(fn ActivateOrDeactivateWindParticle
  (params (bNewActive :type Boolean))
  (if (Variables|Default|GetIs%20Wind%20Enabled)
    (Components|Activation|SetActive (Variables|Effects|GetMuzzleWindParticleSpawned) bNewActive true))
  (return))

(fn ActivateReloadSmokeFX
  (Utilities|FlowControl|Sequence
    (:then_0
      (bind _returnvalue (Niagara|SpawnSystemAttached (Variables|Effects|Reload|GetReload_System) (Variables|Visual|GetWeaponMesh) (Variables|Effects|Reload|GetReloadSocketName) "0, 0, 0" "0, 0, 0" "SnapToTarget" true))
      (Niagara|SetNiagaraVariable(Bool) _returnvalue "UseEmissiveSmoke" (Variables|Effects|Reload|GetUse%20Emissive%20in%20Reload))
      (Niagara|SetNiagaraVariable(LinearColor) _returnvalue "EmissiveColor" (Variables|Effects|Reload|GetReload%20Emissive%20Color))
      (Niagara|SetNiagaraVariable(Float) _returnvalue "RibbonLifetime" (Variables|Effects|Reload|GetReload%20Smoke%20Lifetime))
      (Niagara|SetNiagaraVariable(Float) _returnvalue "RibbonScale" (Variables|Effects|Reload|GetReload%20Smoke%20Width%20Scale))
      (Niagara|SetNiagaraVariable(Material) _returnvalue "RibbonMaterial" (Variables|Effects|Reload|GetReload%20Smoke%20Material))
      (Niagara|SetNiagaraVariable(Bool) _returnvalue "bIsRocketLauncher" (Variables|Effects|Reload|GetIs%20Rocket%20Launcher))
      (Niagara|SetNiagaraVariable(Bool) _returnvalue "HasShellEjects" (Variables|Effects|Reload|GetReload%20Ejects%20Shells))
      (Niagara|SetNiagaraVariable(LinearColor) _returnvalue "BaseColorAlpha" (Variables|Effects|Reload|GetBaseColorAlpha))
      (Niagara|SetNiagaraVariable(LinearColor) _returnvalue "SparksColor" (Variables|Effects|Reload|GetRocket%20Launcher%20Spark%20Color)))
    (:then_1
      (bind _returnvalue_2 (Niagara|SpawnSystemAttached (Variables|Effects|Reload|GetReload_System) (CallFunction|GetLeftHandWeaponMesh) (Variables|Effects|Reload|GetReloadSocketName) "0, 0, 0" "0, 0, 0" "SnapToTarget" true))
      (Niagara|SetNiagaraVariable(Bool) _returnvalue_2 "UseEmissiveSmoke" (Variables|Effects|Reload|GetUse%20Emissive%20in%20Reload))
      (Niagara|SetNiagaraVariable(LinearColor) _returnvalue_2 "EmissiveColor" (Variables|Effects|Reload|GetReload%20Emissive%20Color))
      (Niagara|SetNiagaraVariable(Float) _returnvalue_2 "RibbonLifetime" (Variables|Effects|Reload|GetReload%20Smoke%20Lifetime))
      (Niagara|SetNiagaraVariable(Float) _returnvalue_2 "RibbonScale" (Variables|Effects|Reload|GetReload%20Smoke%20Width%20Scale))
      (Niagara|SetNiagaraVariable(Material) _returnvalue_2 "RibbonMaterial" (Variables|Effects|Reload|GetReload%20Smoke%20Material))
      (Niagara|SetNiagaraVariable(Bool) _returnvalue_2 "bIsRocketLauncher" (Variables|Effects|Reload|GetIs%20Rocket%20Launcher))
      (Niagara|SetNiagaraVariable(Bool) _returnvalue_2 "HasShellEjects" (Variables|Effects|Reload|GetReload%20Ejects%20Shells))
      (Niagara|SetNiagaraVariable(LinearColor) _returnvalue_2 "BaseColorAlpha" (Variables|Effects|Reload|GetBaseColorAlpha))
      (Niagara|SetNiagaraVariable(LinearColor) _returnvalue_2 "SparksColor" (Variables|Effects|Reload|GetRocket%20Launcher%20Spark%20Color)))))

(fn ActivateShellsFX
  (params (Bool :type Boolean) (OffHand :type Boolean))
  (RangedWeaponDual-Effects|NativeActivateShellsFX (Utilities|Struct|MakeShellFXParams (Variables|Effects%28Shells%29|GetShells%20Socket%20Name) (Variables|Effects%28Shells%29|GetShellsParticleSystemTemplate) (Variables|Effects%28Shells%29|GetNiagara%20Shell%20System) (Variables|Effects%28Shells%29|GetShellsRotationRate) (Variables|Effects%28Shells%29|GetShells%20Velocity) (Variables|Effects%28Shells%29|GetShells%20Gravity) (Variables|Effects%28Shells%29|GetShells%20Size) (Variables|Effects%28Shells%29|GetInherit%20Parent%20Velocity) (Variables|Effects%28Shells%29|GetShells%20Lifetime) (Variables|Effects%28Shells%29|GetCylindrical%20Radius) (Variables|Effects%28Shells%29|GetCylindrical%20Height) (Variables|Effects%28Shells%29|GetShells%20Time%20Dilation) (Variables|Effects%28Shells%29|GetShells%20PS%20Max%20Draw%20Distance) (Variables|Effects%28Shells%29|GetShells%20Spawn%20Rate%20Scale) (Variables|Effects%28Shells%29|GetUseShellsOnReload?) Bool OffHand)))

(fn AddRandomScale
  (params (Offhand :type Boolean))
  (bind _returnvalue (Math|Random|RandomFloatinRange 0.5 0.6))
  (bind _returnvalue_2 (Math|Vector|MakeVector _returnvalue _returnvalue _returnvalue))
  (bind _returnvalue_3 (Math|Random|RandomFloatinRange 0.7 1))
  (bind _returnvalue_4 (Math|Vector|MakeVector _returnvalue_3 _returnvalue_3 _returnvalue_3))
  (if (<= (Math|Random|RandomFloatinRange 0.0 1) 0.8)
    (if Offhand
      (Transformation|SetRelativeScale3D (Variables|Default|GetLeftMuzzle(Empty)) _returnvalue_4)
      (else
        (Transformation|SetRelativeScale3D (Variables|Effects|GetMuzzle(Empty)) _returnvalue_4)))
    (elif Offhand
      (Transformation|SetRelativeScale3D (Variables|Default|GetLeftMuzzle(Empty)) _returnvalue_2)
      (else
        (Transformation|SetRelativeScale3D (Variables|Effects|GetMuzzle(Empty)) _returnvalue_2)))))

(fn DeactivateMuzzleFX
  (Components|Activation|Deactivate (Variables|Effects|GetMuzzle(Empty))))

(fn DeactivateReloadSmokeFX)

(fn DeactivateShellsFX
  (RangedWeaponDual-Effects|NativeDeactivateShellsFX))

(fn HideReticle
  (bind _asfort_player_controller_zone (Utilities|Casting|CastToFortPlayerControllerZone (Game|GetInstigatorController)))
  (HUD|SetHUDElementVisibilitywithContext _asfort_player_controller_zone (Variables|Scope|GetReticleHUDElementTags) true "None" ""))

(fn MuzzleFlashFX
  (params (Persistent%20Fire :type Boolean) (OffHand :type Boolean))
  (if (and (not (Weapon|IsUsingScope)) (Effects|AreFireFxRelevant))
    (if (not Persistent%20Fire)
      (CallFunction|AddRandomScale false))
    (if OffHand
      (Components|Activation|Activate (Variables|Default|GetLeftMuzzle(Empty)) true)
      (CallFunction|SetUpMuzzleFXParams (Variables|Default|GetLeftMuzzle%28Empty%29%20Niagara))
      (else
        (Components|Activation|Activate (Variables|Effects|GetMuzzle(Empty)) true)
        (CallFunction|SetUpMuzzleFXParams (Variables|Default|GetMuzzle%28Empty%29%20Niagara))))
    (Variables|Effects|SetLastPlayFXTime (Utilities|Time|GetGameTimeinSeconds))
    (Utilities|Time|ClearandInvalidateTimerbyHandle (Variables|Default|GetDelayedLoopingFXEndHandle)))
  (return))

(fn MuzzlePlayReloadFX
  (params (Selection :type EFortReloadFXState%20Enum :default "ReloadStart"))
  (switch Utilities|FlowControl|Switch|SwitchonEFortReloadFXState Selection
    (:ReloadStart
      ;[1 1 1] Start Reload Smoke
      (if (Variables|Effects|Reload|GetUse%20Reload%20Particles)
        (CallFunction|ActivateReloadSmokeFX)))
    (:Max_None
      ;[1 1 1] Reload Shells
      (CallFunction|ActivateShellsFX (Variables|Effects%28Shells%29|GetUseShellsOnReload?)))))

(fn SetPostProcessParams
  (params (InputPin :type Float%20%28double-precision%29))
  (Variables|Scope|SetScopePostProcessBlendWeight InputPin))

(fn SetUpMuzzleFXParams
  (params (MuzzleSystem :type Niagara%20Particle%20System%20Component%20Object%20Reference))
  (Niagara|SetNiagaraVariable(Vector3) MuzzleSystem "MuzzleCoreScale" (Variables|Effects|Muzzle|GetMuzzle%20Core%20Scale))
  (Niagara|SetNiagaraVariable(Float) MuzzleSystem "MuzzleCoreCapScale" (Variables|Effects|Muzzle|GetMuzzle%20Core%20Cap%20Scale))
  (Niagara|SetNiagaraVariable(Float) MuzzleSystem "SmokeScale" (Variables|Effects|Muzzle|GetSmoke%20Size%20Scale))
  (return))

(fn SetWpnRarity)

(fn ShowReticle
  (bind _asfort_player_controller_zone (Utilities|Casting|CastToFortPlayerControllerZone (Game|GetInstigatorController)))
  (HUD|ResetHUDElementVisibilitytoSettings _asfort_player_controller_zone (Variables|Scope|GetReticleHUDElementTags)))

(fn UpdateShellEmittersFX
  (params (Target :type FXSystem%20Component%20Object%20Reference))
  (RangedWeaponDual-Effects|NativeUpdateShellEmittersFX Target (Variables|Effects%28Shells%29|GetUseShellsOnReload?) (Variables|Effects%28Shells%29|GetShells%20Spawn%20Rate%20Scale)))

(fn ConstructionScript
  (bind _returnvalue (Effects|GetMuzzleSocketName 0))
  (bind _returnvalue_2 (CallFunction|GetLeftHandWeaponMesh))
  (bind _muzzleparticlesystem (Variables|Effects|GetMuzzleParticleSystem))
  (bind _niagara_muzzle_system (Variables|Effects%28Shells%29|GetNiagara%20Muzzle%20System))
  (bind _weaponmesh (Variables|Visual|GetWeaponMesh))
  (Utilities|FlowControl|Sequence
    (:then_0
      ;[0.506 0.05185 1] Set up Muzzle FX
      (Utilities|IsValid (Variables|Effects|GetMuzzleParticleSystem)
        (:Is%20Valid
          (Transformation|AttachComponentToComponent (Variables|Effects|GetMuzzle(Empty)) _weaponmesh _returnvalue "KeepRelative" "KeepRelative" "KeepRelative" true)
          (Transformation|AttachComponentToComponent (Variables|Default|GetLeftMuzzle(Empty)) _returnvalue_2 _returnvalue "KeepRelative" "KeepRelative" "KeepRelative" true)
          (Effects|Components|ParticleSystem|SetTemplate (Variables|Effects|GetMuzzle(Empty)) _muzzleparticlesystem)
          (Effects|Components|ParticleSystem|SetTemplate (Variables|Default|GetLeftMuzzle(Empty)) _muzzleparticlesystem))
        (:Is%20Not%20Valid
          (Utilities|IsValid (Variables|Effects%28Shells%29|GetNiagara%20Muzzle%20System)
            (:Is%20Valid
              (Transformation|AttachComponentToComponent (Variables|Default|GetMuzzle%28Empty%29%20Niagara) _weaponmesh _returnvalue "KeepRelative" "KeepRelative" "KeepRelative" true)
              (Transformation|AttachComponentToComponent (Variables|Default|GetLeftMuzzle%28Empty%29%20Niagara) _returnvalue_2 _returnvalue "KeepRelative" "KeepRelative" "KeepRelative" true)
              (Niagara|SetNiagaraSystemAsset (Variables|Default|GetMuzzle%28Empty%29%20Niagara) _niagara_muzzle_system true)
              (Niagara|SetNiagaraSystemAsset (Variables|Default|GetLeftMuzzle%28Empty%29%20Niagara) _niagara_muzzle_system true))))))
    (:then_1
      ;[1 1 1] Wind Particle Setup
      ;; Set Is Wind Enabled
      (bind _returnvalue_3 (Rendering|Material|GetScalarParameterValue "/Game/Packages/Fortress_SharedMaterials/GlobalMaterialParameters/FortniteMaterialParameters.FortniteMaterialParameters" "Water Interaction Enabled" :!class /Script/Engine.KismetMaterialLibrary))
      (bind _output_get (Variables|Default|SetIs%20Wind%20Enabled (> _returnvalue_3 0.5)))
      (if (not _output_get)
        (Components|DestroyComponent (Variables|Effects|GetMuzzleWindParticleSpawned)))
      (if (Variables|Default|GetIs%20Wind%20Enabled)
        (bind _muzzlewindparticlesystem (Variables|Effects|GetMuzzleWindParticleSystem)
          (:then
            (bind _returnvalue_4 (AddComponent|NODE_AddParticleSystemComponent-1))
            (Effects|Components|ParticleSystem|SetTemplate _returnvalue_4 _muzzlewindparticlesystem)
            (bind _output_get_2 (Variables|Effects|SetMuzzleWindParticleSpawned _returnvalue_4))
            (Transformation|AttachComponentToComponent _output_get_2 (Variables|Visual|GetWeaponMesh) (Effects|GetMuzzleSocketName 0) "KeepRelative" "KeepRelative" "KeepRelative" true))
          (:else))))))
