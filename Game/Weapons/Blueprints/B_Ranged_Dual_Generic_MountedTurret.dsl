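; Generic dual-handed mounted turret weapon Blueprint that drives firing, reload, wind and cosmetic alteration effects, plus overheat state and pierce-through toggling.
; The construction script optionally attaches left and right muzzle and shell emitters, adding wind particles when a global material parameter enables them.
; OnHandFired picks the hand for muzzle flash and shell ejection, with an ammo-based timer as failsafe for stuck looping effects.
; Cosmetic alterations async-load replacement particle systems.
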(event EventBeginPlay
  ;[1 1 1] EVENT -> BeginPlay
  (Rendering|SetVisibility (Variables|Visual|GetWeaponMesh) false false) ; Hack so weapon isn't visibile until in hand
  (bind _asfort_player_pawn (Utilities|Casting|CastToFortPlayerPawn (Game|GetInstigator)))
  (Variables|MountedTurret|SetPlayerPawn _asfort_player_pawn)
  (CallFunction|BindonHandFired)
  ;[1 1 1] Send wind particle reference to the wind manager
  (if (Variables|MountedTurret|GetIs%20Wind%20Enabled)
    (Wind|AddWindParticleSystemComponent (Fortnite|GetWindManager) (Variables|Effects|GetMuzzleWindParticleSpawned))))

(event Custom|BindOnHandFired
  ;[1 1 1] Bind Event to OnHandFired
  (RangedWeaponDual|BindEventtoOnHandFired (delegate Custom|HandFired)))

(event Custom|HandFired
  (params (Hand :type EDualWeaponHand%20Enum :default "LEFT") (bPersistantFire :type Boolean))
  ;[1 1 1] Bind Event to OnHandFired
  (bind _output_get (Variables|MountedTurret|SetLastHandFired Hand))
  (CallFunction|MuzzleFlashFX bPersistantFire (Utilities|Enum|Equal(Enum) _output_get "LEFT"))
  (if (and (not (Weapon|IsUsingScopeorFirstPersonCamera)) (> (Rendering|GetRenderingDetailMode) 0))
    (CallFunction|ActivateShellsFX (Variables|Effects%28Shells%29|GetUseShellsOnFire?) (Utilities|Enum|Equal(Enum) _output_get "LEFT")))
  (FortWeapon|PlayForceFeedback false)
  ;[1 0.09001 0.04025] Failsafe hack to stop looping FX if we exceed the RPC limit and don't properly remove the gameplay cue
  (if (Effects|ShouldPlayPersistentFireFX)
    (Utilities|Time|SetTimerbyFunctionName 0 "StopWeaponFireFX" (* (/ (Math|Conversions|ToFloat(Integer) (FortWeapon|GetLocalMagazineAmmoCount)) (Math|Float|Max(Float) (FortWeapon|GetFiringRate) 0.1)) 1.2))))

(event EventOnPlayWeaponFireFX
  (params (bPersistentFire :type Boolean) (bSecondaryFire :type Boolean))
  ;[1 1 1] EVENT -> PlayWeaponFireFX
  (Utilities|FlowControl|Sequence
    (:then_0
      (CallFunction|ActivateorDeactivateWindParticle true))
    (:then_1
      (FortWeapon|PlayForceFeedback false bPersistentFire))))

(event EventOnStopWeaponFireFX
  (bind _returnvalue (Utilities|Time|GetGameTimeinSeconds))
  ;[1 1 1] EVENT -> StopWeaponFireFX
  (Utilities|Time|ClearTimerbyFunctionName 0 "StopWeaponFireFX")
  ;[1 1 1] Enforce a short minimum time to play looping FX if bPersistentFireFX is true
  (if (not (>= (- _returnvalue (Variables|Effects|GetLastPlayFXTime)) (Variables|Effects|GetMinPlayFXTime)))
    (Utilities|FlowControl|RetriggerableDelay (- (Variables|Effects|GetMinPlayFXTime) (- _returnvalue (Variables|Effects|GetLastPlayFXTime)))))
  ;[1 1 1] Clean up FX if Persistent Fire is on
  (CallFunction|ActivateorDeactivateWindParticle false)
  (CallFunction|DeactivateMuzzleFX)
  (Effects|DeactivateBeamFX)
  (CallFunction|DeactivateShellsFX)
  (FortWeapon|StopForceFeedback))

(event EventOnInitCosmeticAlterations
  (params (CosmeticMod :type Fort%20Cosmetic%20Modification%20Structure))
  (bind (_cosmeticmaterial _ambientparticlesystem _ambientniagarasystem _muzzleparticlesystem _muzzleniagarasystem _reloadparticlesystem _reloadniagarasystem _beamparticlesystem _beamniagarasystem _tracertemplate _bmodifycolor _coloralteration _colorparametername _bmodifydecalcolour _decalcolouralterationstart _decalcolouralterationend _bmodifyshellcolour _shellcolouralteration) (Utilities|Struct|BreakFortCosmeticModification CosmeticMod))
  (bind (_r _g _b) (Utilities|Struct|BreakLinearColor _shellcolouralteration))
  ;[1 1 1] Setup Alteration FX
  ;; Alteration Logic
  (Networking|CanExecuteCosmeticEvents
    (:True
      (Utilities|FlowControl|Sequence
        (:then_0
          (if (not (Asset|IsAssetNull _ambientniagarasystem))
            (bind _object (Utilities|AsyncLoadAsset _ambientniagarasystem)
              (:Completed
                (bind _asniagara_system (Utilities|Casting|CastToNiagaraSystem _object))
                (bind _alteration_ambient_ps (Variables|Effects%28Alterations%29|GetAlteration%20Ambient%20PS)
                  (:then
                    (Components|DestroyComponent _alteration_ambient_ps))
                  (:else
                    _))
                (bind _niagara_component_system_reference (Utilities|ElementalWeaponSetupFX (Variables|Visual|GetWeaponMesh) _asniagara_system "FX_Idle_Start" "FX_Idle_End" false)
                  (:Out
                    (bind _output_get (Variables|Effects%28Alterations%29|SetAlteration%20Ambient%20PS _niagara_component_system_reference))
                    (LOD|SetMaxDrawDistance _output_get 5e+03)))))))
        (:then_1
          (if (not (Asset|IsAssetNull _muzzleparticlesystem))
            (bind _object_2 (Utilities|AsyncLoadAsset _muzzleparticlesystem)
              (:Completed
                (bind _asparticle_system (Utilities|Casting|CastToParticleSystem _object_2))
                (Effects|Components|ParticleSystem|SetTemplate (Variables|Effects|GetMuzzle(Empty)) _asparticle_system)))))
        (:then_2
          (if (and (Effects|GetUseBeamParticles) (not (Asset|IsAssetNull _beamparticlesystem)))
            (bind _object_3 (Utilities|AsyncLoadAsset _beamparticlesystem)
              (:Completed
                (bind _asparticle_system_2 (Utilities|Casting|CastToParticleSystem _object_3))
                (Effects|SetBeamParticleSystem _asparticle_system_2)
                (Effects|Components|ParticleSystem|SetTemplate (CallFunction|GetBeamPSC) _asparticle_system_2)))))
        (:then_3
          (if (not (Asset|IsAssetNull _reloadparticlesystem))
            (Development|PrintString (Utilities|String|Append (Utilities|GetDisplayName self) " - Reload Currently unavailable for Cosmetic Alterations, please file a JIRA with this warning assigning to Eric Ketchum"))))
        (:then_4
          (if _bmodifydecalcolour
            (Decals|SetDecalColorStart _decalcolouralterationstart)
            (Decals|SetDecalColorEnd _decalcolouralterationend)))
        (:then_5
          (if _bmodifyshellcolour
            (Effects|Components|ParticleSystem|SetFloatParameter (Variables|Effects%28Shells%29|GetShells%28empty%29) "Red" _r)
            (Effects|Components|ParticleSystem|SetFloatParameter (Variables|Effects%28Shells%29|GetShells%28empty%29) "Green" _g)
            (Effects|Components|ParticleSystem|SetFloatParameter (Variables|Effects%28Shells%29|GetShells%28empty%29) "Blue" _b)))))))

(event EventOnPlayReloadFX
  (params (ReloadStage :type EFortReloadFXState%20Enum :default "ReloadStart"))
  ;[1 1 1] EVENT -> PlayReloadFX
  (CallFunction|MuzzlePlayReloadFX ReloadStage)
  (FortWeapon|PlayForceFeedback true))

(event EventOnInitAlteration
  (params (NewAlteration :type Fort%20Alteration%20Item%20Definition%20Object%20Reference)))

(event EventOnSetTargeting
  (params (bNewIsTargeting :type Boolean))
  ;[1 1 1] EVENT -> OnSetTargeting (ADS Scope setup)
  (Networking|CanExecuteCosmeticEvents
    (:True
      (bind _asfort_pawn (Utilities|Casting|CastToFortPawn (Game|GetInstigator)))
      (Utilities|IsValid (Pawn|GetLocalViewingPlayerController _asfort_pawn)
        (:Is%20Valid
          (CallFunction|ShowReticle))))))

(event FortWeapon|EventOnEquippedAndAttached
  ;[1 1 1] EVENT -> OnEquip
  (CallFunction|ShowReticle)
  (Utilities|FlowControl|Delay 0.1) ; This is a hack fix
  (CallFunction|SetActiveAlterationIdleParticles true)
  (Utilities|Casting|CastToFortPawn (Game|GetInstigator))
  (Rendering|SetReceivesDecals (CallFunction|GetLeftHandWeaponMesh) false) ; So Sprays don't show up on weapon
)

(event FortWeapon|EventOnLocalUnEquip
  ;[1 1 1] EVENT -> UnEquip
  (CallFunction|SetActiveAlterationIdleParticles false)
  (bind _asfort_pawn (Utilities|Casting|CastToFortPawn (Game|GetInstigator)))
  (Utilities|IsValid (Pawn|GetLocalViewingPlayerController _asfort_pawn)
    (:Is%20Valid
      (CallFunction|ShowReticle))))

(event EventOnWeaponVisibilityChanged
  (params (bVisible :type Boolean) (bSetForLocalControllerOnly :type Boolean))
  (bind _alteration_ambient_ps (Variables|Effects%28Alterations%29|GetAlteration%20Ambient%20PS))
  ;[1 1 1] EVENT -> OnWeaponVisibilityChanged
  (Utilities|IsValid (Variables|Effects%28Alterations%29|GetAlteration%20Ambient%20PS)
    (:Is%20Valid
      (if bVisible
        (Components|Activation|Activate _alteration_ambient_ps false)
        (else
          (Components|Activation|Deactivate _alteration_ambient_ps))))))

(event Custom|Update%20Enemy%20Custom%20Depths
  (params (Enable%20Or%20Disable :type Boolean :default true) (StencilBufferValue :type Integer :default 3))
  ;[1 1 1] Custom Depth
  (bind _outactors (Actor|GetAllActorsOfClass "/Script/FortniteGame.FortAIPawn"))
  (bind _output_get (Variables|Scope|SetArray%20Of%20Active%20Enemy%20AI _outactors))
  (for _array_element _output_get
    (Utilities|IsValid (Variables|/Script/Engine.Character|GetMesh _array_element)
      (:Is%20Valid
        (Rendering|SetRenderCustomDepth (Variables|/Script/Engine.Character|GetMesh _array_element) Enable%20Or%20Disable)
        (Rendering|SetCustomDepthStencilValue (Variables|/Script/Engine.Character|GetMesh _array_element) StencilBufferValue)))))

(event Custom|SetWeaponPierceThrough
  (params (Enable :type Boolean) (TargetLimit :type Integer))
  (bind _returnvalue (Math|Integer|SelectInt TargetLimit 0 Enable))
  ;[1 1 1] Event to dynamically enable/disable trace through pawns
  (FortWeapon|SetTraceThroughPawnsLimit _returnvalue)
  (FortWeapon|SetTraceThroughPawns Enable)
  (CallFunction|SetWeaponPierceThroughClientRep Enable _returnvalue))

(event Custom|SetWeaponPierceThrough_ClientRep
  (params (Enable :type Boolean) (TargetLimit :type Integer))
  ;[1 1 1] Event to dynamically enable/disable trace through pawns
  (FortWeapon|SetTraceThroughPawnsLimit (Math|Integer|SelectInt TargetLimit 0 Enable))
  (FortWeapon|SetTraceThroughPawns Enable))

(event Custom|HideWeaponMesh
  ;[1 1 1] DEBUG EVENTS -> ke * <event name>
  (Development|SetHiddeninGame (Variables|Visual|GetWeaponMesh) true false)
  (Development|PrintString "HideWeaponMesh"))

(event Custom|ShowWeaponMesh
  ;[1 1 1] DEBUG EVENTS -> ke * <event name>
  (Development|SetHiddeninGame (Variables|Visual|GetWeaponMesh) false false)
  (Development|PrintString "ShowWeaponMesh"))

(event Custom|HideWeapon
  ;[1 1 1] DEBUG EVENTS -> ke * <event name>
  (FortWeapon|SetWeaponVisibility false true)
  (Development|PrintString "HideWeapon"))

(event Custom|ShowWeapon
  ;[1 1 1] DEBUG EVENTS -> ke * <event name>
  (FortWeapon|SetWeaponVisibility true true)
  (Development|PrintString "ShowWeapon"))

(event Overheating|EventOnStartCooling
  ;[1 1 1] Store fire state (idle, cooling, heating, overheated)
  (if (not (Utilities|Enum|Equal(Enum) (Variables|Default|GetFireState) "NewEnumerator3"))
    (Variables|Default|SetFireState "NewEnumerator2")))

(event Overheating|EventOnFinishedCooling
  ;[1 1 1] Store fire state (idle, cooling, heating, overheated)
  (Variables|Default|SetFireState "NewEnumerator0"))

(event Overheating|EventOnStartHeating
  ;[1 1 1] Store fire state (idle, cooling, heating, overheated)
  (if (not (Utilities|Enum|Equal(Enum) (Variables|Default|GetFireState) "NewEnumerator3"))
    (Variables|Default|SetFireState "NewEnumerator1")))

(event Overheating|EventOnStartOverheated
  ;[1 1 1] Store fire state (idle, cooling, heating, overheated)
  (Variables|Default|SetFireState "NewEnumerator3"))

(event Overheating|EventOnLocalFailedToFireWhenOverheated
  ;[1 1 1] Fail to fire audio feedback (local only)
  (Audio|PlaySound2D (Variables|Sounds|GetSound_OnFailToFireWhenOverheated)))

(fn ActivateOrDeactivateWindParticle
  (params (bNewActive :type Boolean))
  (if (Variables|MountedTurret|GetIs%20Wind%20Enabled)
    (Components|Activation|SetActive (Variables|Effects|GetMuzzleWindParticleSpawned) bNewActive true))
  (return))

(fn ActivateReloadSmokeFX
  (Utilities|FlowControl|Sequence
    (:then_0
      (bind _returnvalue (Niagara|SpawnSystemAttached (Variables|Effects|Reload|GetReload_System) (Variables|Visual|GetWeaponMesh) (Variables|Effects|Reload|GetReloadSocketName) "0, 0, 0" "0, 0, 0" "SnapToTarget" true))
      (Niagara|SetNiagaraVariable(Bool) _returnvalue "UseEmissiveSmoke" (Variables|Effects|Reload|GetUse%20Emissive%20in%20Reload))
      (Niagara|SetNiagaraVariable(LinearColor) _returnvalue "EmissiveColor" (Variables|Effects|Reload|GetReload%20Emissive%20Color))
      (Niagara|SetNiagaraVariable(Float) _returnvalue "RibbonLifetime" (Variables|Effects|Reload|GetReload%20Smoke%20Lifetime))
      (Niagara|SetNiagaraVariable(Float) _returnvalue "RibbonScale" (Variables|Effects|Reload|GetReload%20Smoke%20Width%20Scale))
      (Niagara|SetNiagaraVariable(Material) _returnvalue "RibbonMaterial" (Variables|Effects|Reload|GetReload%20Smoke%20Material))
      (Niagara|SetNiagaraVariable(Bool) _returnvalue "bIsRocketLauncher" (Variables|Effects|Reload|GetIs%20Rocket%20Launcher))
      (Niagara|SetNiagaraVariable(Bool) _returnvalue "HasShellEjects" (Variables|Effects|Reload|GetReload%20Ejects%20Shells))
      (Niagara|SetNiagaraVariable(LinearColor) _returnvalue "BaseColorAlpha" (Variables|Effects|Reload|GetBaseColorAlpha))
      (Niagara|SetNiagaraVariable(LinearColor) _returnvalue "SparksColor" (Variables|Effects|Reload|GetRocket%20Launcher%20Spark%20Color)))
    (:then_1
      (bind _returnvalue_2 (Niagara|SpawnSystemAttached (Variables|Effects|Reload|GetReload_System) (CallFunction|GetLeftHandWeaponMesh) (Variables|Effects|Reload|GetReloadSocketName) "0, 0, 0" "0, 0, 0" "SnapToTarget" true))
      (Niagara|SetNiagaraVariable(Bool) _returnvalue_2 "UseEmissiveSmoke" (Variables|Effects|Reload|GetUse%20Emissive%20in%20Reload))
      (Niagara|SetNiagaraVariable(LinearColor) _returnvalue_2 "EmissiveColor" (Variables|Effects|Reload|GetReload%20Emissive%20Color))
      (Niagara|SetNiagaraVariable(Float) _returnvalue_2 "RibbonLifetime" (Variables|Effects|Reload|GetReload%20Smoke%20Lifetime))
      (Niagara|SetNiagaraVariable(Float) _returnvalue_2 "RibbonScale" (Variables|Effects|Reload|GetReload%20Smoke%20Width%20Scale))
      (Niagara|SetNiagaraVariable(Material) _returnvalue_2 "RibbonMaterial" (Variables|Effects|Reload|GetReload%20Smoke%20Material))
      (Niagara|SetNiagaraVariable(Bool) _returnvalue_2 "bIsRocketLauncher" (Variables|Effects|Reload|GetIs%20Rocket%20Launcher))
      (Niagara|SetNiagaraVariable(Bool) _returnvalue_2 "HasShellEjects" (Variables|Effects|Reload|GetReload%20Ejects%20Shells))
      (Niagara|SetNiagaraVariable(LinearColor) _returnvalue_2 "BaseColorAlpha" (Variables|Effects|Reload|GetBaseColorAlpha))
      (Niagara|SetNiagaraVariable(LinearColor) _returnvalue_2 "SparksColor" (Variables|Effects|Reload|GetRocket%20Launcher%20Spark%20Color)))))

(fn ActivateShellsFX
  (params (Bool :type Boolean) (OffHand :type Boolean))
  (if Bool
    (if OffHand
      (Components|Activation|Activate (Variables|Default|GetLeftShells(Empty)) true)
      (CallFunction|UpdateShellEmittersFX (Variables|Default|GetLeftShells(Empty)))
      (else
        (Components|Activation|Activate (Variables|Effects%28Shells%29|GetShells%28empty%29) true)
        (CallFunction|UpdateShellEmittersFX (Variables|Effects%28Shells%29|GetShells%28empty%29))))))

(fn AddRandomScale
  (params (Offhand :type Boolean))
  (if (<= (Math|Random|RandomFloatinRange 0.0 1) 0.8)
    (if Offhand
      (Transformation|SetRelativeScale3D (Variables|Default|GetLeftMuzzle(Empty)) (Math|Vector|MakeVector (Math|Random|RandomFloatinRange 0.7 1) (Math|Random|RandomFloatinRange 0.7 1) (Math|Random|RandomFloatinRange 0.7 1)))
      (else
        (Transformation|SetRelativeScale3D (Variables|Effects|GetMuzzle(Empty)) (Math|Vector|MakeVector (Math|Random|RandomFloatinRange 0.7 1) (Math|Random|RandomFloatinRange 0.7 1) (Math|Random|RandomFloatinRange 0.7 1)))))
    (elif Offhand
      (Transformation|SetRelativeScale3D (Variables|Default|GetLeftMuzzle(Empty)) (Math|Vector|MakeVector (Math|Random|RandomFloatinRange 0.5 0.6) (Math|Random|RandomFloatinRange 0.5 0.6) (Math|Random|RandomFloatinRange 0.5 0.6)))
      (else
        (Transformation|SetRelativeScale3D (Variables|Effects|GetMuzzle(Empty)) (Math|Vector|MakeVector (Math|Random|RandomFloatinRange 0.5 0.6) (Math|Random|RandomFloatinRange 0.5 0.6) (Math|Random|RandomFloatinRange 0.5 0.6)))))))

(fn DeactivateMuzzleFX
  (Components|Activation|Deactivate (Variables|Effects|GetMuzzle(Empty))))

(fn DeactivateReloadSmokeFX)

(fn DeactivateShellsFX
  (Components|Activation|Deactivate (Variables|Effects%28Shells%29|GetShells%28empty%29)))

(fn HideReticle
  (bind _asfort_player_controller_zone (Utilities|Casting|CastToFortPlayerControllerZone (Game|GetInstigatorController)))
  (HUD|SetHUDElementVisibilitywithContext _asfort_player_controller_zone (Variables|Scope|GetReticleHUDElementTags) true "None" ""))

(fn MuzzleFlashFX
  (params (Persistent%20Fire :type Boolean) (OffHand :type Boolean))
  ;[1 1 1] If local player is using the scope, don't play muzzle effects
  (if (not (Weapon|IsUsingScope))
    (if (not Persistent%20Fire)
      (CallFunction|AddRandomScale false))
    (if OffHand
      (Components|Activation|Activate (Variables|Default|GetLeftMuzzle(Empty)) true)
      (else
        (Components|Activation|Activate (Variables|Effects|GetMuzzle(Empty)) true)))
    (Variables|Effects|SetLastPlayFXTime (Utilities|Time|GetGameTimeinSeconds)))
  (return))

(fn MuzzlePlayReloadFX
  (params (Selection :type EFortReloadFXState%20Enum :default "ReloadStart"))
  (switch Utilities|FlowControl|Switch|SwitchonEFortReloadFXState Selection
    (:ReloadStart
      ;[1 1 1] Start Reload Smoke
      (if (Variables|Effects|Reload|GetUse%20Reload%20Particles)
        (CallFunction|ActivateReloadSmokeFX)))
    (:Max_None
      ;[1 1 1] Reload Shells
      (CallFunction|ActivateShellsFX (Variables|Effects%28Shells%29|GetUseShellsOnReload?)))))

(fn SetActiveAlterationIdleParticles
  (params (Active :type Boolean) (Reset :type Boolean))
  (Networking|CanExecuteCosmeticEvents
    (:True
      (bind _alteration_ambient_ps (Variables|Effects%28Alterations%29|GetAlteration%20Ambient%20PS)
        (:then
          (Components|Activation|SetActive _alteration_ambient_ps Active Reset)
          (return))
        (:else)))))

(fn SetupFX
  (Transformation|AttachComponentToComponent (Variables|Effects|GetMuzzle(Empty)) (Variables|Visual|GetWeaponMesh) (Effects|GetMuzzleSocketName 0) "KeepRelative" "KeepRelative" "KeepRelative" true)
  (Transformation|AttachComponentToComponent (Variables|Default|GetLeftMuzzle(Empty)) (CallFunction|GetLeftHandWeaponMesh) (Effects|GetMuzzleSocketName 0) "KeepRelative" "KeepRelative" "KeepRelative" true)
  (Effects|Components|ParticleSystem|SetTemplate (Variables|Effects|GetMuzzle(Empty)) (Variables|Effects|GetMuzzleParticleSystem))
  (Effects|Components|ParticleSystem|SetTemplate (Variables|Default|GetLeftMuzzle(Empty)) (Variables|Effects|GetMuzzleParticleSystem))
  ;[1 1 1] Wind Particle Setup
  ;; Set Is Wind Enabled_2
  (bind _returnvalue (Rendering|Material|GetScalarParameterValue "/Game/Packages/Fortress_SharedMaterials/GlobalMaterialParameters/FortniteMaterialParameters.FortniteMaterialParameters" "Water Interaction Enabled" :!class /Script/Engine.KismetMaterialLibrary))
  (bind _output_get (Variables|MountedTurret|SetIs%20Wind%20Enabled (> _returnvalue 0.5)))
  (if (not _output_get)
    (Components|DestroyComponent (Variables|Effects|GetMuzzleWindParticleSpawned)))
  (if (Variables|MountedTurret|GetIs%20Wind%20Enabled)
    (bind _returnvalue_2 (AddComponent|NODE_AddParticleSystemComponent-2))
    (Effects|Components|ParticleSystem|SetTemplate _returnvalue_2 (Variables|Effects|GetMuzzleWindParticleSystem))
    (bind _output_get_2 (Variables|Effects|SetMuzzleWindParticleSpawned _returnvalue_2))
    (Transformation|AttachComponentToComponent _output_get_2 (Variables|Visual|GetWeaponMesh) (Effects|GetMuzzleSocketName 0) "KeepRelative" "KeepRelative" "KeepRelative" true))
  ;[1 1 1] Shells FX setup
  (if (Math|Boolean|ORBoolean (Variables|Effects%28Shells%29|GetUseShellsOnFire?) (Variables|Effects%28Shells%29|GetUseShellsOnReload?) (Variables|Effects%28Shells%29|GetUseShellsOnPump?))
    (CallFunction|SetupShellFX)))

(fn SetupShellFX
  (bind _shells_time_dilation (Variables|Effects%28Shells%29|GetShells%20Time%20Dilation))
  (Effects|Components|ParticleSystem|SetTemplate (Variables|Effects%28Shells%29|GetShells%28empty%29) (Variables|Effects%28Shells%29|GetShellsParticleSystemTemplate))
  (Transformation|AttachComponentToComponent (Variables|Effects%28Shells%29|GetShells%28empty%29) (Variables|Visual|GetWeaponMesh) (Variables|Effects%28Shells%29|GetShells%20Socket%20Name) "KeepRelative" "KeepRelative" "KeepRelative" true)
  (Transformation|AttachComponentToComponent (Variables|Default|GetLeftShells(Empty)) (CallFunction|GetLeftHandWeaponMesh) (Variables|Effects%28Shells%29|GetShells%20Socket%20Name) "KeepRelative" "KeepRelative" "KeepRelative" true)
  (if (Variables|Effects%28Shells%29|GetDebugShellsSocket?)
    (if (Transformation|DoesSocketExist (Variables|Visual|GetWeaponMesh) (Variables|Effects%28Shells%29|GetShells%20Socket%20Name))
      (Development|PrintString "***Success! Socket Found!***" true true "(R=0.000000,G=1.000000,B=0.000000,A=1.000000)" 5)
      (else
        (Development|PrintString "***Error! Cannot find socket name to attach shell eject FX to!***" true true "(R=1.000000,G=0.000000,B=0.000000,A=1.000000)" 5))))
  (Effects|Components|ParticleSystem|SetFloatParameter (Variables|Effects%28Shells%29|GetShells%28empty%29) "SpawnRateScale" (Variables|Effects%28Shells%29|GetShells%20Spawn%20Rate%20Scale))
  (Effects|Components|ParticleSystem|SetVectorParameter (Variables|Effects%28Shells%29|GetShells%28empty%29) "ShellsRotationRate" (Variables|Effects%28Shells%29|GetShellsRotationRate))
  (Effects|Components|ParticleSystem|SetVectorParameter (Variables|Effects%28Shells%29|GetShells%28empty%29) "Velocity" (Variables|Effects%28Shells%29|GetShells%20Velocity))
  (Effects|Components|ParticleSystem|SetVectorParameter (Variables|Default|GetLeftShells(Empty)) "Velocity" (Variables|Effects%28Shells%29|GetLeft%20Shells%20Velocity))
  (Effects|Components|ParticleSystem|SetVectorParameter (Variables|Effects%28Shells%29|GetShells%28empty%29) "Gravity" (Variables|Effects%28Shells%29|GetShells%20Gravity))
  (Effects|Components|ParticleSystem|SetFloatParameter (Variables|Effects%28Shells%29|GetShells%28empty%29) "Lifetime" (Variables|Effects%28Shells%29|GetShells%20Lifetime))
  (Effects|Components|ParticleSystem|SetVectorParameter (Variables|Effects%28Shells%29|GetShells%28empty%29) "Size" (Variables|Effects%28Shells%29|GetShells%20Size))
  (Effects|Components|ParticleSystem|SetVectorParameter (Variables|Effects%28Shells%29|GetShells%28empty%29) "InheritParentVelocity" (Math|Conversions|ToVector(Float) (Variables|Effects%28Shells%29|GetInherit%20Parent%20Velocity)))
  (Effects|Components|ParticleSystem|SetFloatParameter (Variables|Effects%28Shells%29|GetShells%28empty%29) "CylinderRadius" (Variables|Effects%28Shells%29|GetCylindrical%20Radius))
  (Effects|Components|ParticleSystem|SetFloatParameter (Variables|Effects%28Shells%29|GetShells%28empty%29) "CylinderHeight" (Variables|Effects%28Shells%29|GetCylindrical%20Height))
  (Variables|/Script/CascadeRuntime.ParticleSystemComponent|SetCustomTimeDilation _shells_time_dilation (Variables|Effects%28Shells%29|GetShells%28empty%29))
  (Variables|/Script/CascadeRuntime.ParticleSystemComponent|SetCustomTimeDilation _shells_time_dilation (Variables|Default|GetLeftShells(Empty)))
  (switch Utilities|FlowControl|Switch|SwitchonEn_ShellTypes_01 (Variables|Effects%28Shells%29|GetShellTypeSelect)
    (:NewEnumerator0
      (Variables|Effects%28Shells%29|SetSmallShells true))
    (:NewEnumerator1
      (Variables|Effects%28Shells%29|SetMediumShells true))
    (:NewEnumerator2
      (Variables|Effects%28Shells%29|SetLargeShells true))
    (:NewEnumerator3
      (Variables|Effects%28Shells%29|SetShotgunShells true))
    (:NewEnumerator4
      (Variables|Effects%28Shells%29|SetEnergyShells true))))

(fn SetWpnRarity)

(fn ShowReticle
  (bind _asfort_player_controller_zone (Utilities|Casting|CastToFortPlayerControllerZone (Game|GetInstigatorController)))
  (HUD|ResetHUDElementVisibilitytoSettings _asfort_player_controller_zone (Variables|Scope|GetReticleHUDElementTags)))

(fn UpdateShellEmittersFX
  (params (PS :type Cascade%20Particle%20System%20Component%20Object%20Reference))
  (Utilities|FlowControl|Sequence
    (:then_0
      ;[1 1 1] Looping Shell Emitters
      (Effects|Components|ParticleSystem|SetEmitterEnable PS "Small Loop" (and (Variables|Effects%28Shells%29|GetSmallShells) (Effects|ShouldPlayPersistentFireFX)))
      (Effects|Components|ParticleSystem|SetEmitterEnable PS "Medium Loop" (and (Variables|Effects%28Shells%29|GetMediumShells) (Effects|ShouldPlayPersistentFireFX)))
      (Effects|Components|ParticleSystem|SetEmitterEnable PS "Large Loop" (and (Variables|Effects%28Shells%29|GetLargeShells) (Effects|ShouldPlayPersistentFireFX)))
      (Effects|Components|ParticleSystem|SetEmitterEnable PS "Energy Cell Loop" (and (Variables|Effects%28Shells%29|GetEnergyShells) (Effects|ShouldPlayPersistentFireFX)))
      ;[1 1 1] Single Shell Emitters
      (Effects|Components|ParticleSystem|SetEmitterEnable PS "Small Single" (and (Variables|Effects%28Shells%29|GetSmallShells) (not (Effects|ShouldPlayPersistentFireFX))))
      (Effects|Components|ParticleSystem|SetEmitterEnable PS "Medium Single" (and (Variables|Effects%28Shells%29|GetMediumShells) (not (Effects|ShouldPlayPersistentFireFX))))
      (Effects|Components|ParticleSystem|SetEmitterEnable PS "Large Single" (and (Variables|Effects%28Shells%29|GetLargeShells) (not (Effects|ShouldPlayPersistentFireFX))))
      (Effects|Components|ParticleSystem|SetEmitterEnable PS "Shotgun Single" (and (Variables|Effects%28Shells%29|GetShotgunShells) (not (Effects|ShouldPlayPersistentFireFX))))
      (Effects|Components|ParticleSystem|SetEmitterEnable PS "Energy Cell Single" (and (Variables|Effects%28Shells%29|GetEnergyShells) (not (Effects|ShouldPlayPersistentFireFX)))))
    (:then_1
      (if (Variables|Effects%28Shells%29|GetUseShellsOnReload?)
        (Effects|Components|ParticleSystem|SetFloatParameter PS "BurstScale" (Math|Conversions|ToFloat(Integer) (Utilities|ClampMinInteger (CallFunction|GetShellReloadCounter))))
        (CallFunction|SetShellReloadCounter -1)))
    (:then_2
      (return))))

(fn ConstructionScript
  (if (Variables|Default|GetSetupFXInConstructor)
    (CallFunction|SetupFX)))
