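; Weapon component Blueprint that handles firing camera shake for its owning weapon.
; The firing shake event passes the owner weapon to the shared WeaponFunctionLibrary FiringCameraShake function.
; GetShakeDataClass selects between the component's hip-fire and ADS shake data based on a hip-fire flag.
; IsUpdatedFeedbackEnabledOld is hard-coded to return true.
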(event EventCallFiringCameraShake
  (Class|WeaponFunctionLibrary|FiringCameraShake (Variables|Default|GetOwnerWeapon)))

(fn GetShakeDataClass
  (params (bIsHipFire :type Boolean))
  (return (select bIsHipFire (Variables|CameraShake|GetHipfireShakeData) (Variables|CameraShake|GetADSShakeData))))

(fn IsUpdatedFeedbackEnabledOld
  (return true))
