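; Blueprint function library for weapon camera effects, mainly firing camera shake.
; Helpers find or add a camera modifier on the player camera manager, resolved from a weapon's instigator or a pawn.
; FiringCameraShake either calls a native implementation or feeds WeaponCameraAnimation start time, fire-rate-based shot interval and ADS or hipfire data from WeaponCameraAnimationComponent.
; It logs that new implementations belong in that component; HasValidCameraAnimationData has no body.
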(fn FindOrAddCameraModifier
  (params (ModifierClass :type Camera%20Modifier%20Class%20Reference) (InWeapon :type Fort%20Weapon%20Object%20Reference))
  (bind _inweapon (Variables|GetInWeapon))
  (bind _instigator (Variables|/Script/Engine.Actor|GetInstigator _inweapon))
  (Utilities|IsValid _instigator
    (:Is%20Valid
      (bind _asfort_player_pawn (Utilities|Casting|CastToFortPlayerPawn _instigator))
      (bind _asfort_player_controller (Utilities|Casting|CastToFortPlayerController (Pawn|GetController _asfort_player_pawn)))
      (Utilities|IsValid (Variables|/Script/Engine.PlayerController|GetPlayerCameraManager _asfort_player_controller)
        (:Is%20Valid
          (bind _returnvalue (Game|Player|FindCameraModifierbyClass (Variables|/Script/Engine.PlayerController|GetPlayerCameraManager _asfort_player_controller) (Variables|GetModifierClass)))
          (Utilities|IsValid _returnvalue
            (:Is%20Valid
              (LocalVariables|SetOutMod _returnvalue))
            (:Is%20Not%20Valid
              (bind _returnvalue_2 (Game|Player|AddNewCameraModifier (Variables|/Script/Engine.PlayerController|GetPlayerCameraManager _asfort_player_controller) (Variables|GetModifierClass)))
              (LocalVariables|SetOutMod _returnvalue_2)))
          (return (LocalVariables|GetOutMod)))))))

(fn FindOrAddCameraModifierForPawn
  (params (ModifierClass :type Camera%20Modifier%20Class%20Reference) (InPawnActor :type Actor%20Object%20Reference))
  (bind _asfort_player_pawn (Utilities|Casting|CastToFortPlayerPawn InPawnActor))
  (bind _asfort_player_controller (Utilities|Casting|CastToFortPlayerController (Pawn|GetController _asfort_player_pawn)))
  (Utilities|IsValid (Variables|/Script/Engine.PlayerController|GetPlayerCameraManager _asfort_player_controller)
    (:Is%20Valid
      (bind _returnvalue (Game|Player|FindCameraModifierbyClass (Variables|/Script/Engine.PlayerController|GetPlayerCameraManager _asfort_player_controller) (Variables|GetModifierClass)))
      (Utilities|IsValid _returnvalue
        (:Is%20Valid
          (LocalVariables|SetOutMod _returnvalue))
        (:Is%20Not%20Valid
          (bind _returnvalue_2 (Game|Player|AddNewCameraModifier (Variables|/Script/Engine.PlayerController|GetPlayerCameraManager _asfort_player_controller) (Variables|GetModifierClass)))
          (LocalVariables|SetOutMod _returnvalue_2)))
      (return (LocalVariables|GetOutMod)))))

(fn FiringCameraShake
  (params (Weapon :type Fort%20Weapon%20Ranged%20Object%20Reference))
  (bind _weapon (Variables|GetWeapon))
  (bind _returnvalue (Actor|GetComponentbyClass _weapon "/Game/Weapons/Blueprints/WeaponCameraAnimationComponent.WeaponCameraAnimationComponent_C"))
  (bind _activeshake (LocalVariables|GetActiveShake))
  (if (Class|WeaponCameraFunctionLibrary|UseWeaponCameraFunctionLibraryNativeImplementation)
    (Class|WeaponCameraFunctionLibrary|FiringCameraShakeNative Weapon)
    (Fortnite|Utility|FortBlueprintLog 0 "error: Firing Camera Shake got called directly from WeaponFunctionLibrary. New Implementation should be coming from inside a WeaponCameraAnimationComponent.")
    (else
      (Fortnite|Utility|FortBlueprintLog 0 "warning: This Firing Camera Shake call could have been made without a WeaponCameraAnimationComponent. New Implementation should be coming from inside a WeaponCameraAnimationComponent.")
      (Utilities|IsValid _returnvalue
        (:Is%20Valid
          (bind _outmodifier (CallFunction|FindorAddCameraModifier "/Game/Weapons/Blueprints/Cameras/WeaponCameraAnimation.WeaponCameraAnimation_C" Weapon))
          (bind _asweapon_camera_animation (Utilities|Casting|CastToWeaponCameraAnimation _outmodifier))
          (LocalVariables|SetActiveShake _asweapon_camera_animation)
          (Variables|/Game/Weapons/Blueprints/Cameras/WeaponCameraAnimation.WeaponCameraAnimation_C|SetStartTime (Utilities|Time|GetTimeSeconds 0) _activeshake)
          (Variables|/Game/Weapons/Blueprints/Cameras/WeaponCameraAnimation.WeaponCameraAnimation_C|SetTimeBetweenShots (+ (Math|Float|SafeDivide 1.0 (FortWeapon|GetFiringRate (Variables|GetWeapon))) 0.05) _activeshake)
          (Utilities|IsValid _returnvalue
            (:Is%20Valid
              (if (Camera|GetIsTargeting (Variables|GetWeapon))
                (Variables|/Game/Weapons/Blueprints/Cameras/WeaponCameraAnimation.WeaponCameraAnimation_C|SetShakeData (Variables|/Game/Weapons/Blueprints/WeaponCameraAnimationComponent.WeaponCameraAnimationComponent_C|GetADSShakeData _returnvalue) _activeshake)
                (else
                  (Variables|/Game/Weapons/Blueprints/Cameras/WeaponCameraAnimation.WeaponCameraAnimation_C|SetShakeData (Variables|/Game/Weapons/Blueprints/WeaponCameraAnimationComponent.WeaponCameraAnimationComponent_C|GetHipfireShakeData _returnvalue) _activeshake))))))))))

(fn HasValidCameraAnimationData
  (params (Weapon :type Fort%20Weapon%20Object%20Reference)))
