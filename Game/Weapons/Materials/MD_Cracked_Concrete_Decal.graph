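# Deferred decal material for cracked concrete, under weapons, with opacity that changes over a set duration from a Time parameter.
# Road crack masks darken squared concrete color, and road crack normals add detail to tiled concrete normals.
# Opacity combines inverted crack detail and a tiled crack texture, offset by a staggered gradient sampled on inverted normalized elapsed time, then masked by a ground char texture.
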Material /Game/Weapons/Materials/MD_Cracked_Concrete_Decal
graph {
 properties {blendMode=BLEND_Translucent, materialDomain=MD_DeferredDecal, pixelDepthOffsetMode=PDOM_Legacy, usedShadingModels=Default Lit}
 nodes [
  Constant() {r=0.85}
  n1 = Multiply(A=TextureCoordinate()) {constB=2}
  n2 = TextureSample(UVs=@n1) {samplerType=SAMPLERTYPE_Masks, texture=/Game/Athena/Apollo/Environments/Landscape/Textures/Roads/Road_Cracks_M}
  n3 = Multiply(A=TextureCoordinate(), B=ScalarParameter() {defaultValue=3.5, parameterName=Tile})
  n4 = TextureSample(UVs=@n3) {texture=/Game/Environments/Apollo/Sets/Coliseum_Ruins/Textures/T_CR_Concrete_Cracked_D}
  Multiply(
   A=Multiply(A=LinearInterpolate(Alpha=@n2.G) {constA=0.3}, B=Multiply(A=@n4.RGB, B=@n4.RGB))
   B=ScalarParameter() {defaultValue=0.4, parameterName=BaseColorBrightness}
  ) {constB=0.7}
  n5 = TextureSample(UVs=@n1) {samplerType=SAMPLERTYPE_Normal, texture=/Game/Athena/Apollo/Environments/Landscape/Textures/Roads/Road_Cracks_N}
  n6 = Constant3Vector() {constant=<rgba 0.5 0.5 0 0>}
  Multiply(
   A=Add(
    A=TextureSample(UVs=@n3) {samplerType=SAMPLERTYPE_Normal, texture=/Game/Environments/Apollo/Sets/Coliseum_Ruins/Textures/T_CR_Concrete_Cracked_N}.RGB
    B=Multiply(A=@n5.RGB, B=@n6)
   )
   B=VectorParameter() {parameterName=NormalMult, defaultValue=<rgba 3 3 1 1>}.RGB
  )
  n7 = TextureSample() {samplerType=SAMPLERTYPE_LinearColor, texture=/Game/Effects/Fort_Effects/Gameplay/SmokeDustExplosions/Textures/T_GroundChar_01}
  n8 = OneMinus(None=@n2.G)
  n9 = Multiply(A=Multiply(A=@n8, B=@n8), B=ScalarParameter() {defaultValue=0.5, parameterName=DetailCrackMult}) {constB=0.25}
  n10 = TextureSample(UVs=@n3) {samplerType=SAMPLERTYPE_Grayscale, texture=/Game/Effects/Fort_Effects/Textures/Tile/T_Concrete_Crack_A}
  n11 = Divide(A=Subtract(A=Time(), B=ScalarParameter() {parameterName=Time}), B=ScalarParameter() {defaultValue=3, parameterName=Duration})
  n12 = TextureSample(UVs=OneMinus(None=Saturate(None=@n11))) {texture=/Game/Effects/Fort_Effects/Textures/Grad/T_Staggered_Gradient}
  n13 = Saturate(None=Add(A=Add(A=@n9, B=@n10.G), B=Subtract(A=@n12.G)))
  Saturate(None=Multiply(A=Multiply(A=@n7.G, B=@n13), B=ScalarParameter() {defaultValue=10, parameterName=OpacityMult}))
 ]
}
