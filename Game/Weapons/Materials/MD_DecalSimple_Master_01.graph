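# Master deferred decal material for weapon decals with a timed glowing mark, using a ground char texture by default.
# Age over Lifetime, from Spawntime or a looping debug clock, drives an emissive color shift, shrinking radial mask and fade to black.
# The decal texture, scaled by opacity and faded near end of life, is inverted and subtracted to erode the glow.
# A static switch toggles this emissive, which feeds a shared FX emissive-adjustment function.
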Material /Game/Weapons/Materials/MD_DecalSimple_Master_01
graph {
 properties {
  bAutomaticallySetUsageInEditor=false
  bUsedWithNanite=true
  blendMode=BLEND_Translucent
  materialDomain=MD_DeferredDecal
  pixelDepthOffsetMode=PDOM_Legacy
  shadingModel=MSM_Unlit
  usedShadingModels=Unlit
 }
 nodes [
  n1 = Time()
  n2 = Time()
  n3 = ScalarParameter() {group=Timestamp, parameterName=Spawntime}
  n4 = Divide(
   A=StaticSwitchParameter(True=Frac(None=@n1), False=Subtract(A=@n2, B=@n3)) {group=Timestamp, parameterName=Debug Timestamp}
   B=ScalarParameter() {defaultValue=5, group=Timestamp, parameterName=Lifetime}
  )
  n5 = Clamp(None=Multiply(A=@n4, B=ScalarParameter() {defaultValue=6, group=Emissive, parameterName=Emissive Transition Multiplier}) {constB=6})
  n6 = LinearInterpolate(
   A=VectorParameter() {group=Emissive, parameterName=Emissive Color A, defaultValue=<rgba 8 6.0936 2.64 1>}.RGB
   B=VectorParameter() {group=Emissive, parameterName=Emissive Color B, defaultValue=<rgba 1 0.1 0.01 1>}.RGB
   Alpha=@n5
  )
  n7 = SphereMask(
   A=TextureCoordinate()
   B=Constant2Vector() {g=0.5, r=0.5}
   Radius=LinearInterpolate(
    A=ScalarParameter() {defaultValue=0.4, group=Emissive, parameterName=Radius A}
    B=ScalarParameter() {defaultValue=0.1, group=Emissive, parameterName=Radius B}
    Alpha=@n5
   )
   Hardness=Constant()
  )
  n8 = TextureSampleParameter2D() {
   parameterName=DecalTexture
   samplerType=SAMPLERTYPE_LinearColor
   texture=/Game/Effects/Fort_Effects/Gameplay/SmokeDustExplosions/Textures/T_GroundChar_01
  }
  n9 = ScalarParameter() {defaultValue=0.9, parameterName=Opacity}
  n10 = Multiply(A=OneMinus(None=@n4), B=ScalarParameter() {defaultValue=4, group=Timestamp, parameterName=Fade Transition Multiplier}) {constB=4}
  n11 = OneMinus(None=LinearInterpolate(B=LinearInterpolate(B=@n8.RGB, Alpha=@n9) {constB=0}, Alpha=Clamp(None=@n10)) {constB=0})
  n12 = ScalarParameter() {defaultValue=0.55, parameterName=Erode Color Scale}
  n13 = Clamp(None=Subtract(A=Multiply(A=@n6, B=@n7), B=Multiply(A=@n11, B=@n12)))
  MaterialFunctionCall(Emissive=StaticSwitchParameter(True=LinearInterpolate(A=@n13, Alpha=@n5) {constB=0}, False=Constant()) {defaultValue=true, group=Timestamp, parameterName=Use Emissive?}) {materialFunction=&MF_FX_EmissiveAdjustments}
  ScalarParameter() {parameterName=Metallic}
  ScalarParameter() {parameterName=Specular}
  ScalarParameter() {defaultValue=1, parameterName=Roughness}
  VectorParameter() {parameterName=DecalColor, defaultValue=<rgba 0 0 0 1>}
 ]
}
