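# Deferred decal material for a scorch mark with glowing embers that fade over the decal's lifetime, with a normal map and constant roughness and specular.
# Embers blend an emissive mask, panning cube noise and a normal-map dot against radial UV direction, then a rising threshold erodes them and drives edge-to-center color.
# Scorch opacity erodes through its mask on separate timing, with optional facing fade.
# Static switches add decal-color tinting and looping debug animation.
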Material /Game/Weapons/Materials/MD_Decal_Scorch_Embers_Normal
graph {
 properties {
  bAutomaticallySetUsageInEditor=false
  blendMode=BLEND_Translucent
  materialDomain=MD_DeferredDecal
  pixelDepthOffsetMode=PDOM_Legacy
  shadingModel=MSM_Unlit
  usedShadingModels=Unlit
 }
 nodes [
  n1 = TextureSampleParameter2D() {
   parameterName=Normal
   samplerType=SAMPLERTYPE_Normal
   texture=/Game/Effects/Fort_Effects/Gameplay/SmokeDustExplosions/Textures/T_Decal_Scorch_A_N
  }
  MaterialFunctionCall(Normal=@n1.RGB, Flatness=ScalarParameter() {parameterName=NormalFlatness}) {materialFunction=&FlattenNormal}
  PixelNormalWS()
  ScalarParameter() {defaultValue=0.3, parameterName=Specular}
  ScalarParameter() {defaultValue=1, parameterName=Roughness}
  n2 = DecalColor()
  n3 = ScalarParameter() {defaultValue=0.05, parameterName=EmissiveFadeStartOffset}
  n4 = Constant() {r=8}
  n5 = Fmod(A=Time(), B=@n4)
  n6 = StaticSwitchParameter(True=OneMinus(None=Divide(A=@n5, B=@n4)), False=DecalLifetimeOpacity().Opacity) {parameterName=DebugAnim}
  n7 = Saturate(None=LinearInterpolate(A=ScalarParameter() {defaultValue=1, parameterName=EmissiveFadeEndOffset}, B=Multiply(A=@n3) {constB=-1}, Alpha=@n6) {constA=-7})
  n8 = OneMinus(None=@n7)
  n9 = Power(Base=@n7) {constExponent=3}
  n10 = Multiply(A=Subtract(A=TextureCoordinate()) {constB=0.5}) {constB=-2}
  n11 = ScalarParameter() {defaultValue=3, parameterName=EmissiveNormalIntensity}
  n12 = DotProduct(A=@n1.RGB, B=AppendVector(A=Multiply(A=@n10, B=@n11), B=ScalarParameter() {defaultValue=0.2, parameterName=EmissiveNormalFlatness}))
  n13 = TextureSampleParameter2D() {
   parameterName=EmissiveMask
   samplerType=SAMPLERTYPE_LinearColor
   texture=/Game/Effects/Fort_Effects/Gameplay/SmokeDustExplosions/Textures/T_GroundChar_01_E
  }
  n14 = ChannelMaskParameter(None=@n13.RGB) {maskChannel=Green, parameterName=EmissiveMaskChannel, defaultValue=<rgba 0 1 0 0>}
  n15 = ScalarParameter() {defaultValue=0.6, parameterName=NoiseAdd}
  n16 = ScalarParameter() {defaultValue=0.5, parameterName=NoiseTile}
  n17 = TextureCoordinate()
  n18 = Multiply(A=Time(), B=ScalarParameter() {defaultValue=0.25, parameterName=PanSpeed})
  n19 = TextureSample(UVs=Panner(Coordinate=Multiply(A=@n16, B=@n17), Time=@n18) {speedX=0.005, speedY=-0.02}) {samplerType=SAMPLERTYPE_Masks, texture=/Game/Global/Textures/Default/Noise/T_CubeNoise}
  n20 = Panner(Coordinate=Multiply(A=@n16, B=TextureCoordinate() {uTiling=1.5, vTiling=1.5}), Time=@n18) {speedX=-0.015, speedY=0.03}
  n21 = Multiply(
   A=@n19.R
   B=Multiply(A=@n19.B, B=TextureSample(UVs=@n20) {samplerType=SAMPLERTYPE_Masks, texture=&T_CubeNoise}.B)
  )
  n22 = ScalarParameter() {defaultValue=1, parameterName=NoiseMult}
  n23 = Multiply(A=@n14.R, B=Add(A=Multiply(A=@n14.R, B=@n15), B=Multiply(A=@n21, B=@n22) {constB=4}) {constB=0.7})
  n24 = SmoothStep(
   Min=@n9
   Max=Add(A=@n9, B=ScalarParameter() {defaultValue=0.1, parameterName=SmoothStepSize}) {constB=0.1}
   Value=Power(Base=LinearInterpolate(A=@n12, B=@n23)) {constExponent=3}
  )
  n25 = Multiply(
   A=LinearInterpolate(
    A=VectorParameter() {group=Emissive, parameterName=EmissiveEdge, defaultValue=<rgba 0.8 0.016 0 1>}.RGB
    B=VectorParameter() {group=Emissive, parameterName=EmissiveCenter, defaultValue=<rgba 3 0.611 0.0705 1>}.RGB
    Alpha=Multiply(A=@n8, B=@n24)
   )
   B=@n24
  )
  n26 = Saturate(None=@n25)
  n27 = MaterialFunctionCall(Emissive=StaticSwitchParameter(
    True=MaterialFunctionCall(LDR Input=@n26, HDR Tint=@n25) {materialFunction=&DeriveHDRfromLDR}.RemappedRange
    False=Constant()
   ) {defaultValue=true, group=Timestamp, parameterName=Use Emissive}) {materialFunction=&MF_FX_EmissiveAdjustments}
  StaticSwitchParameter(True=Multiply(A=@n2.RGB, B=@n27.Result), False=@n27.Result) {parameterName=UseDecalColor}
  n28 = ScalarParameter() {defaultValue=2, parameterName=ScorchFadeDelay}
  n29 = Saturate(None=LinearInterpolate(B=Add(A=@n28), Alpha=@n6))
  n30 = StaticSwitchParameter(
   True=Multiply(A=MaterialFunctionCall() {materialFunction=/Game/Weapons/Prototype/SprayGun/Materials/MF_DecalFacingMask}.Result, B=@n29)
   False=@n29
  ) {group=DotSurfaceFade, parameterName=UseDotSurfaceFade, sortPriority=0}
  n31 = OneMinus(None=@n30)
  n32 = Add(A=@n31) {constB=0.5}
  n33 = ChannelMaskParameter(None=TextureSampleParameter2D() {parameterName=OpacityMask, texture=/Game/Effects/Fort_Effects/Gameplay/SmokeDustExplosions/Textures/T_Decal_Scorch_A}.G) {maskChannel=Green, parameterName=OpacityMaskChannel, defaultValue=<rgba 0 1 0 0>}
  n34 = Multiply(
   A=NamedRerouteDeclaration(None=SmoothStep(Min=@n31, Max=@n32, Value=@n33.R)) {name=ScorchMask, nodeColor=<rgba 1 0 0.2109 1>}
   B=Multiply(A=@n30, B=ScalarParameter() {defaultValue=0.9, parameterName=OpacityMult})
  )
  StaticSwitchParameter(True=Multiply(A=@n34, B=@n2.A), False=@n34) {parameterName=UseDecalColor}
  VectorParameter() {parameterName=DecalColor, defaultValue=<rgba 0 0 0 1>}
 ]
}
